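# Sound cue for the B52 glider's deploy (open) sound in the Athena cosmetics set.
# A Random node picks one of three equally weighted open wave variations.
# Playback uses the shared glider open/close attenuation and concurrency, a glider-open audio property sheet, and the Fort_Gliders_Land sound class.
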SoundCue /BRCosmetics/Sounds/Gliders/B52/Athena_Glider_B52_Open_Cue
root=Random_1
nodes [
 {id=Random_1, children [WavePlayer_0,WavePlayer_6,WavePlayer_7], properties {weights [1,1,1]}}
 {id=WavePlayer_0, properties {soundWaveAssetPtr=&/SFX_Glider_B52_Open_01}}
 {id=WavePlayer_6, properties {soundWaveAssetPtr=&/SFX_Glider_B52_Open_02}}
 {id=WavePlayer_7, properties {soundWaveAssetPtr=&/SFX_Glider_B52_Open_03}}
]
properties {
 attenuationSettings=/Game/Sounds/Attenuation/Fort_Player_Glider_OpenClose_Athena_Att
 audioPropertiesSheet=/BRCosmetics/Sounds/Settings/PropertySheets/Glider/APS_Cosmetic_Glider_Open
 maxDistance=10000
 soundClassObject=&Fort_Gliders_Land
 volumeMultiplier=1
 concurrencySet [&Athena_GliderOpenClose_Conc]
}
